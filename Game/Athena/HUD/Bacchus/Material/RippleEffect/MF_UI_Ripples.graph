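# UI material function in the Bacchus HUD that generates animated ripple rings from a Phase input and a gradient shape, defaulting to radial distance from UV center.
# Phase minus the gradient raised to Distribution is divided by Spacing, then clamped to half the floored RippleCount to limit the ring count.
# An absolute sine turns this into bands, offset by squared OutlineThickness.
# Outputs the soft Gradient and a Solid version sharpened by a narrow smoothstep.
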MaterialFunction /Game/Athena/HUD/Bacchus/Material/RippleEffect/MF_UI_Ripples
graph {nodes [
  n1 = FunctionInput() {
   bUsePreviewValueAsDefault=true
   description=The step of the effect. Plug Time in to animate.
   inputName=Phase
   inputType=FunctionInput_Scalar
   sortPriority=2
   previewValue=<1 0 0 1>
  }
  n2 = FunctionInput(Preview=TextureCoordinate()) {bUsePreviewValueAsDefault=true, inputName=UV, inputType=FunctionInput_Vector2, sortPriority=1}
  n3 = Constant() {r=0.5}
  n4 = FunctionInput(Preview=Multiply(A=Distance(A=@n2, B=@n3)) {constB=2}) {
   bUsePreviewValueAsDefault=true
   description=The shape of the effect. From 0 - 1.
   inputName=Gradient
   inputType=FunctionInput_Scalar
   sortPriority=1
  }
  n5 = FunctionInput() {bUsePreviewValueAsDefault=true, inputName=Distribution, inputType=FunctionInput_Scalar, sortPriority=4, previewValue=<1 0 0 0>}
  n6 = FunctionInput() {bUsePreviewValueAsDefault=true, inputName=Spacing, inputType=FunctionInput_Scalar, sortPriority=5, previewValue=<0.5 0 0 0>}
  n7 = Divide(A=Subtract(A=@n1, B=Power(Base=@n4, Exp=@n5) {constExponent=1.5}) {constB=0}, B=@n6) {constB=0.4}
  n8 = FunctionInput() {bUsePreviewValueAsDefault=true, inputName=RippleCount, inputType=FunctionInput_Scalar, sortPriority=3, previewValue=<3 0 0 0>}
  n9 = Multiply(A=Floor(None=Max(A=@n8) {constB=0})) {constB=0.5, desc=Dividing by 2 to compensate for the abs node}
  n10 = Abs(None=Sine(None=Clamp(None=@n7, Max=@n9)))
  n11 = FunctionInput() {bUsePreviewValueAsDefault=true, inputName=OutlineThickness, inputType=FunctionInput_Scalar, sortPriority=6, previewValue=<0.1 0 0 0>}
  n12 = Multiply(A=@n11, B=@n11)
  n13 = FunctionOutput(None=Add(A=Add(A=@n10, B=@n12) {constB=-0.4}) {constB=-0.5}) {outputName=Gradient, sortPriority=1}
  FunctionOutput(None=Saturate(None=SmoothStep(Value=@n13) {constMax=0.51, constMin=0.49})) {outputName=Solid}
 ]}
