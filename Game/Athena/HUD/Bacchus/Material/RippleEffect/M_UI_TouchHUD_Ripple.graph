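# Translucent UI material that draws animated, tinted circular ripples for the touch HUD.
# A time chain with start time, offset, speed and optional looping or reversal sets the phase, with a manual debug override.
# MF_UI_Ripples builds a procedural distance field from UVs, a radial gradient and that phase, which MF_UI_AntiAliasing thresholds.
# Opacity multiplies the result by color alpha and a start/end size-and-fade mask, which a debug switch can show alone.
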Material /Game/Athena/HUD/Bacchus/Material/RippleEffect/M_UI_TouchHUD_Ripple
graph {
 properties {bAutomaticallySetUsageInEditor=false, blendMode=BLEND_Translucent, materialDomain=MD_UI, usedShadingModels=Default Lit}
 nodes [
  n1 = ScalarParameter() {defaultValue=1, group=Opacity, parameterName=EndSize, sliderMax=1}
  n2 = Subtract(A=Add(A=NamedRerouteUsage()), B=@n1)
  n3 = ScalarParameter() {defaultValue=1, group=Opacity, parameterName=EndFade, sliderMax=1}
  n4 = Saturate(None=Divide(A=OneMinus(None=@n2), B=Multiply(A=@n3, B=@n1)))
  n5 = ScalarParameter() {group=Opacity, parameterName=StartSize, sliderMax=1}
  n6 = OneMinus(None=Subtract(A=NamedRerouteUsage(), B=@n5))
  n7 = ScalarParameter() {defaultValue=0.5, group=Opacity, parameterName=StartFade, sliderMax=1}
  n8 = Saturate(None=Divide(A=Add(A=@n6) {constB=-1}, B=Multiply(A=@n7, B=@n5)))
  NamedRerouteDeclaration(None=Saturate(None=Subtract(A=@n4, B=@n8))) {name=OpacityMask, nodeColor=<rgba 1 0.1172 0 1>}
  n9 = NamedRerouteUsage()
  n10 = Subtract(A=ScalarParameter() {defaultValue=3, group=Animation, parameterName=LoopDuration, sortPriority=4}, B=@n9)
  NamedRerouteDeclaration(None=StaticSwitchParameter(
    True=ScalarParameter() {defaultValue=-2, group=Animation, parameterName=Phase}
    False=StaticSwitchParameter(True=@n10, False=@n9) {group=Animation, parameterName=Reverse}
   ) {group=Debug, parameterName=ManualPhase}) {name=Phase, nodeColor=<rgba 0 0.8275 1 1>}
  n11 = NamedRerouteDeclaration(None=TextureCoordinate()) {name=UV, nodeColor=<rgba 0.9688 1 0 1>}
  n12 = Constant() {r=0.5}
  NamedRerouteDeclaration(None=Multiply(A=Distance(A=@n11, B=@n12)) {constB=2}) {name=Circle, nodeColor=<rgba 1 0 0.1172 1>}
  n13 = Subtract(A=Time() {period=1}, B=ScalarParameter() {group=Animation, parameterName=AnimStartTime, sortPriority=0}) {constA=38.8}
  n14 = ScalarParameter() {group=Animation, parameterName=StartOffset}
  n15 = Multiply(
   A=Add(A=Add(A=@n13, B=@n14), B=ScalarParameter() {group=Opacity, parameterName=StartSize, sliderMax=1})
   B=ScalarParameter() {defaultValue=1, group=Animation, parameterName=Speed, sortPriority=1}
  )
  n16 = Fmod(
   A=@n15
   B=Multiply(
    A=ScalarParameter() {defaultValue=3, group=Animation, parameterName=LoopDuration, sortPriority=4}
    B=ScalarParameter() {defaultValue=1, group=Animation, parameterName=Speed, sortPriority=1}
   )
  )
  n17 = ScalarParameter() {defaultValue=1, group=Animation, parameterName=Loop, sliderMax=1, sortPriority=3}
  n18 = OneMinus(None=ScalarParameter() {defaultValue=1, group=Animation, parameterName=Loop, sliderMax=1, sortPriority=3})
  NamedRerouteDeclaration(None=Add(
    A=Multiply(A=@n16, B=@n17) {desc="Outputs Time when Loop parameter is 1 (true).\r\nOutputs 0 otherwise."}
    B=Multiply(A=@n15, B=@n18) {desc="Outputs Time when Loop parameter is 0 (false).\r\nOutputs 0 otherwise."}
   )) {name=Time, nodeColor=<rgba 0.3634 0.2387 1 1>}
  n19 = NamedRerouteUsage()
  n20 = MaterialFunctionCall(
   DistanceField=MaterialFunctionCall!(
    UV=NamedRerouteUsage()
    Gradient=NamedRerouteUsage()
    Phase=NamedRerouteUsage()
    RippleCount=ScalarParameter() {defaultValue=3, group=Ripples, parameterName=RippleCount}
    Distribution=ScalarParameter() {defaultValue=1, group=Animation, parameterName=Distribution, sortPriority=2}
    Spacing=ScalarParameter() {defaultValue=0.5, group=Ripples, parameterName=Spacing}
    OutlineThickness=ScalarParameter() {defaultValue=0.1, group=Ripples, parameterName=OutlineThickness, sliderMax=1}
   ) {materialFunction=&/MF_UI_Ripples}.Gradient
   Threshold=Constant() {r=0.5}
  ) {materialFunction=&MF_UI_AntiAliasing}
  StaticSwitchParameter(
   True=@n19
   False=Multiply(A=Multiply(A=@n19, B=@n20.Outer), B=VectorParameter() {group=Ripples, parameterName=Color, defaultValue=<rgba 1 1 1 1>}.A)
  ) {group=Debug, parameterName=ShowOpacity}
  VectorParameter() {group=Ripples, parameterName=Color, defaultValue=<rgba 1 1 1 1>}
 ]
}
